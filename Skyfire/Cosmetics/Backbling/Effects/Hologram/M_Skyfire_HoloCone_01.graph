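# Unlit, two-sided translucent particle material for the Skyfire backbling's hologram cone effect.
# Opacity combines a UV height gradient, panning noise varied per face side, inverted dust and an axis-locked fresnel built from local Z and camera direction.
# Color sums two panning caustics and dust, tinted by fresnel and noise-driven colors over a height gradient.
# Particle color scales both; depth and near-camera fades soften opacity.
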Material /Skyfire/Cosmetics/Backbling/Effects/Hologram/M_Skyfire_HoloCone_01
graph {
 properties {
  bUseTranslucencyVertexFog=false
  bUsedWithMeshParticles=true
  bUsedWithNiagaraMeshParticles=true
  bUsedWithParticleSprites=true
  blendMode=BLEND_Translucent
  pixelDepthOffsetMode=PDOM_Legacy
  shadingModel=MSM_Unlit
  translucencyPass=MTP_BeforeDOF
  twoSided=true
  usedShadingModels=Unlit
 }
 nodes [
  n1 = NamedRerouteUsage()
  n2 = ScalarParameter() {defaultValue=0.2}
  n3 = ComponentMask(None=Add(A=Multiply(A=@n1, B=@n2) {constB=0.1}, B=TextureCoordinate())) {g=true}
  n4 = Power(Base=@n3, Exp=ScalarParameter() {defaultValue=3, parameterName=HeightGrad_Power})
  n5 = TextureCoordinate()
  n6 = LinearInterpolate(Alpha=TwoSidedSign()) {constA=0.85, constB=1.15}
  n7 = AppendVector(
   A=ScalarParameter() {defaultValue=3, parameterName=Noise_Tiling_X}
   B=Multiply(A=ScalarParameter() {defaultValue=0.01, parameterName=Noise_Tiling_Y}, B=@n6)
  )
  n8 = AppendVector(
   A=ScalarParameter() {defaultValue=0.025, parameterName=Noise_Panning_X}
   B=ScalarParameter() {defaultValue=0.15, parameterName=Noise_Panning_Y}
  )
  n9 = TextureSample(UVs=Panner(Coordinate=Multiply(A=@n5, B=@n7), Speed=Multiply(A=@n6, B=@n8))) {samplerSource=SSM_Wrap_WorldGroupSettings, samplerType=SAMPLERTYPE_Grayscale, texture=&LowResBlurredNoise}
  n10 = LinearInterpolate(
   A=ScalarParameter() {defaultValue=0.05, parameterName=Noise_AlphaMin}
   Alpha=Power(
    Base=NamedRerouteDeclaration(None=@n9.R) {name=Noise, nodeColor=<rgba 0 1 0.9837 1>}
    Exp=ScalarParameter() {defaultValue=1.5, parameterName=Noise_Exp_Alpha_Inside}
   )
  )
  n11 = Power(Base=@n3, Exp=ScalarParameter() {defaultValue=20, parameterName=BaseAdd_Power})
  n12 = ScalarParameter() {defaultValue=0.25, parameterName=BaseAdd}
  n13 = Clamp(None=Add(A=Multiply(A=@n11, B=@n12), B=Multiply(A=@n4, B=@n10)) {constB=0})
  n14 = Subtract(A=WorldPosition().XYZ, B=CameraPositionWS())
  n15 = Transform(None=Constant3Vector() {constant=<rgba 0 0 1 0>}) {transformSourceType=TRANSFORMSOURCE_Local}
  n16 = Normalize(VectorInput=CrossProduct(A=Normalize(VectorInput=@n14), B=@n15))
  n17 = DotProduct(A=Normalize(VectorInput=CrossProduct(A=@n16, B=@n15)), B=PixelNormalWS())
  n18 = Power(Base=@n17, Exp=ScalarParameter() {defaultValue=2.5, parameterName=Fresnel Exponent})
  n19 = ParticleColor()
  n20 = TextureSampleParameter2D(UVs=Panner(Coordinate=TextureCoordinate() {uTiling=2, vTiling=0.025}) {speedX=0.025, speedY=0.03}) {parameterName=StaticNoise, texture=/Game/Global/Textures/Default/PointSplatter/T_Tiling_Dust}
  n21 = Multiply(A=Multiply(A=Multiply(A=@n13, B=@n18), B=@n19.A), B=Clamp(None=OneMinus(None=@n20.R)))
  n22 = ScalarParameter() {defaultValue=25, parameterName=Fade Distance}
  n23 = ScalarParameter() {defaultValue=150, parameterName=NearCameraFade_Distance}
  Multiply(
   A=Multiply(
    A=DepthFade(Opacity=@n21, FadeDistance=@n22)
    B=MaterialFunctionCall(Distance=@n23) {materialFunction=&Near_Camera_Fade}.Result
   )
   B=ScalarParameter() {defaultValue=0.35, parameterName=Opacity}
  )
  NamedRerouteDeclaration(None=@n17) {name=Fresnel, nodeColor=<rgba 0.0078 0 1 1>}
  NamedRerouteDeclaration(None=@n20.R) {name=StaticNoise-Distortion, nodeColor=<rgba 0.9453 0 1 1>}
  n24 = ScalarParameter() {defaultValue=1, parameterName=CausticA_Tiling_X}
  n25 = ScalarParameter() {defaultValue=0.96, parameterName=CausticA_Tiling_Y}
  n26 = Panner(
   Coordinate=Multiply(A=TextureCoordinate(), B=AppendVector(A=@n24, B=@n25))
   Speed=AppendVector(
    A=ScalarParameter() {defaultValue=-0.11, parameterName=CausticA_Panning_X}
    B=ScalarParameter() {defaultValue=0.6, parameterName=CausticA_Panning_Y}
   )
  ) {speedX=-0.11, speedY=0.6}
  n27 = ScalarParameter() {defaultValue=1, parameterName=CausticB_Tiling_X}
  n28 = ScalarParameter() {defaultValue=1.23, parameterName=CausticB_Tiling_Y}
  n29 = Panner(
   Coordinate=Multiply(A=TextureCoordinate(), B=AppendVector(A=@n27, B=@n28))
   Speed=AppendVector(
    A=ScalarParameter() {defaultValue=0.1, parameterName=CausticB_Panning_X}
    B=ScalarParameter() {defaultValue=0.5, parameterName=CausticB_Panning_Y}
   )
  ) {speedX=0.1, speedY=0.5}
  n30 = Multiply(
   A=Add(
    A=TextureSample(UVs=@n26) {samplerSource=SSM_Wrap_WorldGroupSettings, texture=&T_ShieldCaustics}.R
    B=TextureSample(UVs=@n29) {samplerSource=SSM_Wrap_WorldGroupSettings, texture=&T_ShieldCaustics}.R
   )
   B=ScalarParameter() {defaultValue=1, parameterName=Caustic_Intensity}
  )
  n31 = TextureSampleParameter2D(UVs=Panner(Coordinate=TextureCoordinate() {uTiling=2, vTiling=0.25}) {speedX=0.025, speedY=0.1}) {parameterName=StaticNoise, samplerSource=SSM_Wrap_WorldGroupSettings, texture=&T_Tiling_Dust}
  n32 = ScalarParameter() {defaultValue=1.5, parameterName=Param_1}
  n33 = Multiply(A=@n32) {constB=2}
  n34 = ComponentMask(None=TextureCoordinate()) {g=true}
  n35 = Multiply(
   A=Power(Base=@n31.R, Exp=LinearInterpolate(A=@n33, B=@n32, Alpha=@n34))
   B=ScalarParameter() {defaultValue=8, parameterName=StaticNoise_Intensity}
  )
  n36 = VectorParameter() {parameterName=Bottom Color, defaultValue=<rgba 0.1 1 1 1>}
  n37 = VectorParameter() {parameterName=Top Color, defaultValue=<rgba 0.6667 0 2 1>}
  n38 = Clamp(None=Power(Base=NamedRerouteUsage(), Exp=ScalarParameter() {defaultValue=1.5, parameterName=Noise_Exp}))
  n39 = Multiply(
   A=Add(A=Add(A=@n30, B=@n35))
   B=LinearInterpolate(
    A=VectorParameter() {parameterName=FresnelColor, defaultValue=<rgba 0 1.6667 5 1>}.RGB
    B=LinearInterpolate(A=@n36.RGB, B=@n37.RGB, Alpha=@n38)
    Alpha=NamedRerouteUsage()
   )
  )
  n40 = Multiply(
   A=Power(Base=@n34, Exp=ScalarParameter() {defaultValue=2, parameterName=ColorGrad_Exp})
   B=ScalarParameter() {defaultValue=2, parameterName=BaseColor_Intensity}
  )
  Reroute(None=Multiply(A=Multiply(A=@n39, B=@n40), B=@n19.RGB))
 ]
}
