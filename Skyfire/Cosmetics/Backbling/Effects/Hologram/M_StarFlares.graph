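# Unlit additive material for the star flares on the Skyfire backbling hologram, usable on Niagara mesh particles.
# SF_Rune_Count thresholds the second UV channel's vertical axis in fifths, and ForBlinkingStarOnly blends from a dim fill to a mask isolating the current step.
# That mask multiplies a flare texture and particle alpha.
# Particle color feeds a shared emissive-adjustment function, alongside rune on/off color, outline and opacity branches with no shown output connections.
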Material /Skyfire/Cosmetics/Backbling/Effects/Hologram/M_StarFlares
graph {
 properties {
  bUsedWithNiagaraMeshParticles=true
  blendMode=BLEND_Additive
  pixelDepthOffsetMode=PDOM_Legacy
  shadingModel=MSM_Unlit
  usedShadingModels=Unlit
 }
 nodes [
  n1 = VectorParameter() {group=SkyFire, parameterName=SF_Rune_Color off, defaultValue=<rgba 0 0 0 1>}
  n2 = VectorParameter() {group=SkyFire, parameterName=SF_Rune_Color 0n, defaultValue=<rgba 0.0937 0 0.5312 1>}
  n3 = TextureCoordinate() {coordinateIndex=1}
  n4 = Add(A=ComponentMask(None=TextureCoordinate() {coordinateIndex=1}) {g=true}) {constB=-1}
  n5 = ScalarParameter() {group=SkyFire, parameterName=SF_Rune_Count}
  n6 = Add(A=Divide(A=Subtract(A=@n5)) {constB=5}) {constB=0.2}
  n7 = Sign(None=Clamp(None=Add(A=@n4, B=@n6)))
  n8 = Add(A=ComponentMask(None=@n3) {g=true}) {constB=-1}
  n9 = ScalarParameter() {group=SkyFire, parameterName=SF_Rune_Count}
  n10 = Add(A=@n8, B=Add(A=Divide(A=@n9) {constB=5}) {constB=0.2})
  n11 = Subtract(A=ScalarParameter() {group=SkyFire, parameterName=SF_Rune_Count})
  n12 = Add(A=@n8, B=Add(A=Divide(A=@n11) {constB=5}) {constB=0.2})
  n13 = OneMinus(None=Saturate(None=Divide(A=@n12) {constB=0}))
  n14 = Saturate(None=LinearInterpolate(A=Multiply(A=@n7) {constB=0.1}, B=Multiply(A=@n10, B=@n13), Alpha=ScalarParameter() {defaultValue=1, parameterName=ForBlinkingStarOnly}))
  n15 = Multiply(A=TextureSample(UVs=@n3) {texture=&/T_FX_Skyfire_Backpack_Flares}.R, B=@n14)
  LinearInterpolate(A=@n1.RGB, B=@n2.RGB, Alpha=@n15)
  LinearInterpolate(A=@n1.RGB, B=@n2.RGB, Alpha=@n14)
  n16 = ParticleColor()
  MaterialFunctionCall(Emissive=@n16.RGB) {materialFunction=&MF_FX_EmissiveAdjustments}
  Multiply(A=@n16.A, B=@n15)
  Multiply(A=@n15, B=ScalarParameter() {defaultValue=1, group=SkyFire, parameterName=SF_Holo_Opacity})
  Multiply(A=@n2.RGB, B=ScalarParameter() {defaultValue=2, group=SkyFire, parameterName=SF_Rune_outline multplyer}) {constB=4}
 ]
}
