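# Niagara system for the Skyfire backbling hologram: a glowing, loot-tinted effect attached to a skeletal mesh.
# One burst particle persists past its lifetime, follows the sampled mesh position with an offset, and feeds sprite, light, mesh and component renderers.
# User Alpha and an optional BlinkSpeed-driven sine drive airglow alpha and light intensity, which also flickers over emitter age.
# Alpha also sets the visibility tag; several older color, scale and rotation modules are disabled.
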NiagaraSystem /Skyfire/Cosmetics/Backbling/Effects/Hologram/NS_Skyfire_Hologram
dependencies {usedDataInterfaces [&NiagaraDataInterfaceCurve], usedDynamicInputs [], usedModules [&SystemState]}
emitters {Hologram {
  emitterUpdateScript [
   &EmitterState()
   &SpawnBurst_Instantaneous(
    Spawn Count=&MakeIntFromFloat(Float=1)
    Spawn Probability=HlslExpression() {hlslExpression="Emitter.LoopCount == 0"}
    Use Spawn Probability=true
   )
   /SetVariables(Emitter.OpacityOverride:NiagaraFloat=&Multiply_Float(
     A=0.5
     B=&Add_Float(
      A=1
      B=&Sine(
       Normalized Angle=&Multiply_Float(A=$Engine.Time, B=&Multiply_Float(A=$User.BlinkSpeed, B=1))
       Period=1
       Scale=1
       Bias=0
      )
     )
    )) {isSetParameters=true}
  ]
  particleSpawnScript [
   /Niagara/Modules/Spawn/Initialization/InitializeParticle(
    Color=&MultiplyLinearColor(Linear Color=$User.LootLevel_Color, Scale Factor=<1 1 1 0>)
    Lifetime=0.5
    Mesh Scale=&VectorFromFloat(Value=0.09)
    Module.Write Color=true
    Module.Write Lifetime=true
    Module.Write Mass=true
    Module.Write Material Random=false
    Module.Write Position=true
    Module.Write Scale=true
    Module.Write SpriteRotation=false
    Module.Write SpriteSize=true
    Position=&Add_Vector(A=/Niagara/DynamicInputs/Helpers/SimulationPosition(), B=<0 0 0>)
    Sprite Size=&Vector2DFromFloat(Value=80)
   )
   &SystemLocation(MaskSpawnGroup=false) {enabled=false}
   /SetVariables(Particles.LightColor=<rgba 0 4 3.8 1>) {enabled=false, isSetParameters=true}
   /SetVariables(Particles.LightExponent=2) {enabled=false, isSetParameters=true}
   /SetVariables(Particles.MeshOrientation=&OrientMeshToVector(Look At Direction=<0 0 0>, Reference Axis Vector=<1 0 0>)) {enabled=false, isSetParameters=true}
   /SetVariables(Particles.TextCameraOffset=4) {enabled=false, isSetParameters=true}
   &InitialMeshOrientation(Orientation Axis=<0 0 1>, Orientation Vector=<0 0 0>, Rotation=<0 0 0.25>, Use Orientation Vector=true, Use Rotation Vector=true)
  ]
  particleUpdateScript [
   &ParticleState(Kill Particles When Lifetime Has Elapsed=false, Let Infinitely Lived Particles Die When Emitter Deactivates=true)
   &SkeletalMeshLocation(
    MaskSpawnGroup=false
    Mesh Sampling Type=::NewEnumerator5
    Orientation Sampling=::NewEnumerator1
    OverrideRandomSeed=false
    Sampled Position Offset=<0 -10 8>
    Skeletal Mesh=$User.SkeletalMesh
   )
   &SolveForcesAndVelocity(Clamp Velocity=false, Limit Acceleration=false) {enabled=false}
   /SetVariables(Particles.RotationLimited:Vector3f=&self:LimitVelocity(
     Max:Vector3f=&VectorFromFloat(Value=0.4)
     Min:Vector3f=&VectorFromFloat(Value=-0.4)
    )) {enabled=false, isSetParameters=true}
   /SetVariables(Particles.RotationalVelocity:Vector3f=$Particles.RotationLimited) {enabled=false, isSetParameters=true}
   &ScaleColor(
    Scale Alpha=&FloatFromCurve(
     FloatCurve=DataInterface() {propertyValues {Curve {keys [{interpMode=RCIM_Cubic, tangentMode=RCTM_User, time=0, value=0}, <at 0.7526 0.9779>, <at 1 1>]}}}
     Scale Curve=$Emitter.OpacityOverride
    )
    ScaleA=true
    ScaleRGB=true
   ) {enabled=false}
   /SetVariables(Particles.Color-Crosshair:LinearColor=&MakeLinearColorFromVectorAndFloat(
     "Vector (RGB)"=&MakeVectorFromLinearColorRGB(Color=<rgba 1 0 0 1>)
     "Float (Alpha)"=&FloatFromCurve(
      FloatCurve=DataInterface() {propertyValues {Curve {keys [<at 0 0>,<at 0.25 1>,<at 0.5 0.3>,<at 1 0.3>]}}}
      CurveIndex=0
      Scale Curve=&FloatFromCurve(FloatCurve=DataInterface() {propertyValues {Curve {keys [<at 0 0>,<at 1 1>]}}})
     )
    )) {enabled=false, isSetParameters=true}
   /SetVariables(Particles.SpriteSize-Crosshair:Vector2f=&Vector2DFromFloat(Value=&FloatFromCurve(FloatCurve=DataInterface() {propertyValues {Curve {keys [<at 0 52>,<at 0.25 70>,<at 0.5 60>,<at 1 60>]}}}, CurveIndex=0))) {enabled=false, isSetParameters=true}
   &ScaleMeshSize(Scale Factor=&VectorFromFloat(Value=&FloatFromCurve(FloatCurve=DataInterface() {propertyValues {Curve {keys [
          {arriveTangent=0.6541, interpMode=RCIM_Cubic, leaveTangent=0.6541, tangentMode=RCTM_User, time=0, value=0.85}
          {arriveTangent=-0.0078, interpMode=RCIM_Cubic, leaveTangent=-0.0078, tangentMode=RCTM_User, time=0.35, value=1}
         ]}}}))) {enabled=false}
   &FN_Light_Attributes(
    Color=&MultiplyLinearColor(Linear Color=$User.LootLevel_Color, Scale Factor=<1 1 1 1>)
    Create Color Attribute=true
    Create Exponent Attribute=true
    Create Position Attribute=true
    Create Radius Attribute=true
    Enable Particle Light Enable=true
    Exponent=5
    Light Enabled=true
    Radius=80
    Scale Color=&Multiply_Float(
     A=&Add_Float(
      A=&Multiply_Float(
       A=&Multiply_Float(
        A=0.5
        B=&MakeCustomFloatFromBool(Bool=&SetBoolByFloatComparison(A=$User.TurnOnBlink, B=0), True Float=$Emitter.OpacityOverride, False Float=1)
       )
       B=&ClampFloat(Float=$User.Alpha)
      )
      B=0.1
     )
     B=&Lerp_Float(A=11.5, B=12, Alpha=&Sine(Normalized Angle=$Emitter.Age, Period=0.25, Scale=2, Bias=4))
    )
   )
   &FN_Gen5_Light(
    Attenutation=80
    Color:LinearColor=&MultiplyLinearColor(Linear Color=$User.LootLevel_Color, Scale Factor=<1 1 1 1>)
    Exponent=5
    Intensity:NiagaraFloat=&Multiply_Float(
     A=&Add_Float(
      A=&Multiply_Float(
       A=&Multiply_Float(
        A=0.5
        B=&MakeCustomFloatFromBool(Bool=&SetBoolByFloatComparison(A=$User.TurnOnBlink, B=0), True Float=$Emitter.OpacityOverride, False Float=1)
       )
       B=&ClampFloat(Float=$User.Alpha)
      )
      B=0.1
     )
     B=&Lerp_Float(A=11.5, B=12, Alpha=&Sine(Normalized Angle=$Emitter.Age, Period=0.25, Scale=2, Bias=4))
    )
   )
   /SetVariables(
    Particles.AirglowColor:LinearColor=&MultiplyLinearColor(
     Linear Color=$Particles.Color
     Scale Factor=&MakeVector4(
      X=1
      Y=1
      Z=1
      W=&Add_Float(
       A=&Multiply_Float(
        A=&Multiply_Float(
         A=0.5
         B=&MakeCustomFloatFromBool(Bool=&SetBoolByFloatComparison(A=$User.TurnOnBlink, B=0), True Float=$Emitter.OpacityOverride, False Float=1)
        )
        B=&ClampFloat(Float=$User.Alpha)
       )
       B=0.1
      )
     )
    )
    Particles.AirglowPos:Vector3f=&Add_Vector(A=Linked() {linkedVariable {name=Particles.Position, type=NiagaraPosition}}, B=<0 0 10>)
   ) {isSetParameters=true}
   /SetVariables(Particles.VisTag:NiagaraInt32=&MakeIntFromFloat(Float=$User.Alpha)) {isSetParameters=true}
   &CameraOffset(Camera Offset Amount=20)
  ]
  rendererClasses [
   &NiagaraSpriteRendererProperties
   &NiagaraLightRendererProperties
   &NiagaraMeshRendererProperties
   &NiagaraComponentRendererProperties
  ]
  renderers [
   {rendererClass=&NiagaraSpriteRendererProperties}
   {rendererClass=&NiagaraLightRendererProperties, rendererIndex=1}
   {rendererClass=&NiagaraMeshRendererProperties, rendererIndex=2}
   {rendererClass=&NiagaraComponentRendererProperties, rendererIndex=3}
  ]
 }}
summary {
 systemName=NS_Skyfire_Hologram
 emitters [{
   emitterName=Hologram
   rendererClasses [
    &NiagaraSpriteRendererProperties
    &NiagaraLightRendererProperties
    &NiagaraMeshRendererProperties
    &NiagaraComponentRendererProperties
   ]
  }]
 userVariables [
  User.Alpha {type=NiagaraFloat, defaultValue=0, description=""}
  User.BlinkSpeed {type=NiagaraFloat, defaultValue=1, description=""}
  User.LootLevel {type=NiagaraFloat, defaultValue=0, description=""}
  User.LootLevel_Color {type=LinearColor, description="", defaultValue {struct=LinearColor, value=<rgba 0.8994 0.0377 2 1>}}
  User.SkeletalMesh {
   type=NiagaraDataInterfaceSkeletalMesh
   defaultValue {
    struct=NiagaraExt_VariableValue_DataInterface
    value {dataInterface=&self:NiagaraDataInterfaceSkeletalMesh_0}
   }
  }
  User.TurnOnBlink {type=NiagaraFloat, defaultValue=0, description=""}
 ]
}
